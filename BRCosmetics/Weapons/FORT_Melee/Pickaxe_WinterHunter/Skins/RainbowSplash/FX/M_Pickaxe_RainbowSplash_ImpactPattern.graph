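# Translucent unlit Niagara sprite material for the RainbowSplash pickaxe impact pattern, using a skull texture by default.
# Texture RGB times particle color feeds the shared emissive adjustment function.
# Texture and particle alpha are masked by a sine band along V, offset and scaled by dynamic Fade Start and Fade Length, enabling a wipe.
# A quality switch skips depth fade on Low and Medium; the Tiling parameter is unused.
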Material /BRCosmetics/Weapons/FORT_Melee/Pickaxe_WinterHunter/Skins/RainbowSplash/FX/M_Pickaxe_RainbowSplash_ImpactPattern
graph {
 properties {
  bAutomaticallySetUsageInEditor=false
  bCastRayTracedShadows=false
  bUsedWithNiagaraSprites=true
  blendMode=BLEND_Translucent
  shadingModel=MSM_Unlit
  usedShadingModels=Unlit
 }
 nodes [
  n1 = TextureSampleParameter2D() {parameterName=Tex, texture=&/T_RainbowSplash_Skull}
  n2 = ParticleColor()
  MaterialFunctionCall(Emissive=Multiply(A=@n1.RGB, B=@n2.RGB)) {materialFunction=&MF_FX_EmissiveAdjustments}
  n3 = ComponentMask(None=TextureCoordinate()) {g=true}
  n4 = DynamicParameter() {defaultValue=<rgba 0 1 1 1>, paramNames [Fade Start, Fade Length, Tiling, Param4]}
  n5 = Saturate(None=Multiply(A=Add(A=@n3, B=@n4.Fade Start), B=@n4.Fade Length) {constB=3})
  n6 = Multiply(A=@n1.A, B=Multiply(A=@n2.A, B=Sine(None=@n5) {period=2}))
  n7 = DepthFade(Opacity=@n6) {fadeDistanceDefault=10}
  QualitySwitch(Default=@n7, Low=@n6, High=@n7, Medium=@n6, Epic=@n7)
 ]
}
